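# Physics asset for the VoyagerCuboid backpack cosmetic, giving capsule bodies to a cube bone, many small rock bones and four rock base bones.
# Each rock is constrained to one base bone with identical tight swing and twist limits, allowing only slight movement.
# Most rocks hang off the first base, with smaller groups on the other three.
# The cube body and the bases have no constraints linking them to each other.
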PhysicsAsset /BRCosmetics/Accessories/FORT_Backpacks/Backpack_M_MED_VoyagerCuboid/Meshes/M_MED_VoyagerCuboid_Pack_Physics
bodies {
 dyn_cube_01 [{
   shapeType=Capsule
   length=4.7932
   radius=4.6744
   shapeName=dyn_cube_01_capsule
   center=<0.0012 -0.0003 -0.0032>
   rotation=<rot 6.2412 -90 -101.0851>
  }]
 dyn_rock_01 [{
   shapeType=Capsule
   length=3.5355
   radius=1.264
   shapeName=dyn_rock_01_capsule
   center=<3.5521 -0.3523 0.3698>
   rotation=<rot 0 -86.7794 -86.7913>
  }]
 dyn_rock_02 [{
   shapeType=Capsule
   length=5.3442
   radius=1.554
   shapeName=dyn_rock_02_capsule
   center=<5.1919 0.2194 -0.2059>
   rotation=<rot 0 -92.5487 -89.203>
  }]
 dyn_rock_03 [{
   shapeType=Capsule
   length=1.4264
   radius=0.6283
   shapeName=dyn_rock_03_capsule
   center=<1.4074 0.1186 0.1021>
   rotation=<rot 0 91.0428 -87.7153>
  }]
 dyn_rock_04 [{
   shapeType=Capsule
   length=1.7065
   radius=0.7308
   shapeName=dyn_rock_04_capsule
   center=<1.7131 -0.3493 -0.019>
   rotation=<rot 0 -87.5275 -84.1427>
  }]
 dyn_rock_05 [{
   shapeType=Capsule
   length=4.9683
   radius=1.5055
   shapeName=dyn_rock_05_capsule
   center=<4.9454 -0.6988 0.211>
   rotation=<rot 0 -88.4512 -87.0084>
  }]
 dyn_rock_06 [{
   shapeType=Capsule
   length=6.2993
   radius=1.583
   shapeName=dyn_rock_06_capsule
   center=<6.2346 0.1678 0.1397>
   rotation=<rot 0 90.6588 -89.5708>
  }]
 dyn_rock_07 [{
   shapeType=Capsule
   length=4.5164
   radius=1.4418
   shapeName=dyn_rock_07_capsule
   center=<4.4654 0.2648 -0.2515>
   rotation=<rot 0 -89.7621 -88.4278>
  }]
 dyn_rock_08 [{
   shapeType=Capsule
   length=6.0842
   radius=1.7681
   shapeName=dyn_rock_08_capsule
   center=<5.9763 -0.0891 -0.1153>
   rotation=<rot 0 89.9058 -88.9126>
  }]
 dyn_rock_09 [{
   shapeType=Capsule
   length=4.5612
   radius=1.4227
   shapeName=dyn_rock_09_capsule
   center=<4.4767 -0.1583 0.0083>
   rotation=<rot 0 89.172 -89.629>
  }]
 dyn_rock_10 [{
   shapeType=Capsule
   length=4.2707
   radius=1.068
   shapeName=dyn_rock_10_capsule
   center=<4.1429 0.3763 0.1432>
   rotation=<rot 0 89.4645 -87.2768>
  }]
 dyn_rock_11 [{
   shapeType=Capsule
   length=3.2349
   radius=1.2412
   shapeName=dyn_rock_11_capsule
   center=<3.2038 -0.0466 0.0717>
   rotation=<rot 0 91.2199 -88.5037>
  }]
 dyn_rock_12 [{
   shapeType=Capsule
   length=4.4732
   radius=1.814
   shapeName=dyn_rock_12_capsule
   center=<4.3645 0.4792 0.2479>
   rotation=<rot 0 90.0129 -89.3447>
  }]
 dyn_rock_13 [{
   shapeType=Capsule
   length=5.1203
   radius=1.6648
   shapeName=dyn_rock_13_capsule
   center=<4.6383 -0.2888 0.2855>
   rotation=<rot 0 -88.2812 -86.1439>
  }]
 dyn_rock_14 [{
   shapeType=Capsule
   length=2.2444
   radius=1.0336
   shapeName=dyn_rock_14_capsule
   center=<2.2672 -0.1747 -0.5244>
   rotation=<rot 0 85.4734 -82.7077>
  }]
 dyn_rock_15 [{
   shapeType=Capsule
   length=5.896
   radius=1.9626
   shapeName=dyn_rock_15_capsule
   center=<5.8388 0.016 -0.2673>
   rotation=<rot 0 -91.8358 -89.4519>
  }]
 dyn_rock_16 [{
   shapeType=Capsule
   length=4.5328
   radius=1.2604
   shapeName=dyn_rock_16_capsule
   center=<4.4734 -0.3512 0.2131>
   rotation=<rot 0 88.0159 -87.4056>
  }]
 dyn_rock_17 [{
   shapeType=Capsule
   length=3.3379
   radius=1.2229
   shapeName=dyn_rock_17_capsule
   center=<2.8623 0.888 0.2897>
   rotation=<rot 0 -96.5936 -87.927>
  }]
 dyn_rock_18 [{
   shapeType=Capsule
   length=3.9167
   radius=1.1752
   shapeName=dyn_rock_18_capsule
   center=<3.7166 -0.1644 -0.1299>
   rotation=<rot 0 -90.5969 -88.2423>
  }]
 dyn_rock_19 [{
   shapeType=Capsule
   length=6.4207
   radius=1.6614
   shapeName=dyn_rock_19_capsule
   center=<6.2913 0.8039 0.1844>
   rotation=<rot 0 -90.1762 -89.2004>
  }]
 dyn_rock_20 [{
   shapeType=Capsule
   length=4.7738
   radius=1.2878
   shapeName=dyn_rock_20_capsule
   center=<4.6797 0.4018 -0.0158>
   rotation=<rot 0 -89.5746 -89.1023>
  }]
 dyn_rock_21 [{
   shapeType=Capsule
   length=4.2564
   radius=1.0756
   shapeName=dyn_rock_21_capsule
   center=<4.1923 0.2434 -0.5071>
   rotation=<rot 0 -93.0926 -89.2352>
  }]
 dyn_rock_22 [{
   shapeType=Capsule
   length=9.3059
   radius=3.3729
   shapeName=dyn_rock_22_capsule
   center=<5.8693 0.9736 -1.2109>
   rotation=<rot 0 -85.7144 -88.4998>
  }]
 dyn_rock_23 [{
   shapeType=Capsule
   length=4.1774
   radius=1.3719
   shapeName=dyn_rock_23_capsule
   center=<3.9474 0.7355 -0.0205>
   rotation=<rot 0 -89.5894 -88.934>
  }]
 dyn_rock_24 [{
   shapeType=Capsule
   length=4.4176
   radius=1.6892
   shapeName=dyn_rock_24_capsule
   center=<4.0958 -0.3196 0.0895>
   rotation=<rot 0 83.5433 -88.9236>
  }]
 dyn_rock_25 [{
   shapeType=Capsule
   length=2.575
   radius=1.0717
   shapeName=dyn_rock_25_capsule
   center=<2.6798 -0.5229 0.0065>
   rotation=<rot 0 99.4386 -85.2457>
  }]
 dyn_rock_26 [{
   shapeType=Capsule
   length=3.1691
   radius=1.3373
   shapeName=dyn_rock_26_capsule
   center=<3.1261 -0.3179 0.3883>
   rotation=<rot 0 87.3704 -89.9529>
  }]
 dyn_rock_27 [{
   shapeType=Capsule
   length=3.7667
   radius=1.5269
   shapeName=dyn_rock_27_capsule
   center=<3.7157 0.3216 -0.0913>
   rotation=<rot 0 -85.2428 -87.1398>
  }]
 dyn_rock_28 [{
   shapeType=Capsule
   length=1.176
   radius=0.505
   shapeName=dyn_rock_28_capsule
   center=<1.1916 -0.1682 -0.0419>
   rotation=<rot 0 92.8101 -81.9494>
  }]
 dyn_rock_29 [{
   shapeType=Capsule
   length=2.7489
   radius=0.9826
   shapeName=dyn_rock_29_capsule
   center=<2.6823 0.3235 0.2516>
   rotation=<rot 0 -92.0417 -87.0311>
  }]
 dyn_rock_30 [{
   shapeType=Capsule
   length=2.597
   radius=0.8937
   shapeName=dyn_rock_30_capsule
   center=<2.5968 -0.5891 0.3671>
   rotation=<rot 0 -89.6479 -87.1101>
  }]
 dyn_rock_31 [{
   shapeType=Capsule
   length=7.2255
   radius=2.3733
   shapeName=dyn_rock_31_capsule
   center=<7.1486 -0.4713 -0.0644>
   rotation=<rot 0 89.2745 -89.9128>
  }]
 dyn_rock_32 [{
   shapeType=Capsule
   length=9.1634
   radius=2.723
   shapeName=dyn_rock_32_capsule
   center=<5.4658 -0.4755 -0.5695>
   rotation=<rot 0 -88.7718 -84.269>
  }]
 dyn_rock_33 [{
   shapeType=Capsule
   length=8.0047
   radius=2.2586
   shapeName=dyn_rock_33_capsule
   center=<7.5308 0.3404 -0.0715>
   rotation=<rot 0 -88.7169 -89.5547>
  }]
 dyn_rock_34 [{
   shapeType=Capsule
   length=3.5102
   radius=1.2488
   shapeName=dyn_rock_34_capsule
   center=<3.5001 -0.4002 0.4667>
   rotation=<rot 0 -87.3967 -86.9206>
  }]
 dyn_rock_35 [{
   shapeType=Capsule
   length=3.7505
   radius=1.441
   shapeName=dyn_rock_35_capsule
   center=<3.647 -0.1292 -0.2036>
   rotation=<rot 0 91.2646 -87.4352>
  }]
 dyn_rock_36 [{
   shapeType=Capsule
   length=2.9852
   radius=0.8343
   shapeName=dyn_rock_36_capsule
   center=<2.9798 0.1321 -0.417>
   rotation=<rot 0 88.9276 -87.49>
  }]
 dyn_rock_37 [{
   shapeType=Capsule
   length=3.1305
   radius=1.354
   shapeName=dyn_rock_37_capsule
   center=<2.8004 0.1024 -0.2136>
   rotation=<rot 0 -88.3852 -82.2264>
  }]
 dyn_rock_38 [{
   shapeType=Capsule
   length=7.2534
   radius=4.0323
   shapeName=dyn_rock_38_capsule
   center=<6.6648 -0.0597 1.9867>
   rotation=<rot 0 -89.6649 -86.7658>
  }]
 dyn_rock_39 [{
   shapeType=Capsule
   length=5.2119
   radius=1.5669
   shapeName=dyn_rock_39_capsule
   center=<5.1545 -0.3817 -0.1407>
   rotation=<rot 0 -90.6136 -88.3667>
  }]
 dyn_rock_40 [{
   shapeType=Capsule
   length=3.5058
   radius=0.9886
   shapeName=dyn_rock_40_capsule
   center=<3.4354 0.1898 0.0838>
   rotation=<rot 0 87.3535 -87.9525>
  }]
 dyn_rock_41 [{
   shapeType=Capsule
   length=4.3431
   radius=1.4478
   shapeName=dyn_rock_41_capsule
   center=<3.9206 -0.3049 -0.0617>
   rotation=<rot 0 -85.8854 -88.6668>
  }]
 dyn_rock_42 [{
   shapeType=Capsule
   length=5.4206
   radius=1.556
   shapeName=dyn_rock_42_capsule
   center=<5.3685 0.0239 -0.0144>
   rotation=<rot 0 90.6639 -88.7873>
  }]
 dyn_rock_43 [{
   shapeType=Capsule
   length=2.4758
   radius=1.1254
   shapeName=dyn_rock_43_capsule
   center=<2.4528 -0.1634 0.0844>
   rotation=<rot 0 90.9131 -88.4058>
  }]
 dyn_rock_44 [{
   shapeType=Capsule
   length=2.3358
   radius=1.6111
   shapeName=dyn_rock_44_capsule
   center=<2.5384 0.2426 -0.7295>
   rotation=<rot 0 90.1666 -77.1269>
  }]
 dyn_rock_45 [{
   shapeType=Capsule
   length=1.754
   radius=0.6475
   shapeName=dyn_rock_45_capsule
   center=<1.7455 -0.0693 -0.1537>
   rotation=<rot 0 93.0007 -89.0457>
  }]
 dyn_rock_46 [{
   shapeType=Capsule
   length=3.1282
   radius=1.153
   shapeName=dyn_rock_46_capsule
   center=<2.743 0.0203 0.0746>
   rotation=<rot 0 87.7857 -89.6064>
  }]
 dyn_rock_47 [{
   shapeType=Capsule
   length=0.9878
   radius=0.505
   shapeName=dyn_rock_47_capsule
   center=<0.9764 -0.2348 -0.0456>
   rotation=<rot 0 88.6167 -86.9475>
  }]
 dyn_rock_48 [{
   shapeType=Capsule
   length=1.7371
   radius=0.6635
   shapeName=dyn_rock_48_capsule
   center=<1.7196 -0.418 0.031>
   rotation=<rot 0 89.9959 -89.1499>
  }]
 dyn_rock_49 [{
   shapeType=Capsule
   length=3.0214
   radius=1.2812
   shapeName=dyn_rock_49_capsule
   center=<2.8213 -0.1053 -0.1144>
   rotation=<rot 0 -87.5656 -86.2184>
  }]
 dyn_rock_50 [{
   shapeType=Capsule
   length=5.0719
   radius=1.6877
   shapeName=dyn_rock_50_capsule
   center=<4.9788 0.4825 -0.4184>
   rotation=<rot 0 85.5437 -84.3456>
  }]
 dyn_rock_51 [{
   shapeType=Capsule
   length=2.6719
   radius=0.9377
   shapeName=dyn_rock_51_capsule
   center=<2.6921 0.3261 -0.044>
   rotation=<rot 0 84.6651 -87.3998>
  }]
 dyn_rock_52 [{
   shapeType=Capsule
   length=1.8891
   radius=0.7981
   shapeName=dyn_rock_52_capsule
   center=<1.7868 0.1541 0.0971>
   rotation=<rot 0 94.256 -86.0783>
  }]
 dyn_rock_53 [{
   shapeType=Capsule
   length=2.4453
   radius=1.0474
   shapeName=dyn_rock_53_capsule
   center=<2.4244 0.3551 0.1983>
   rotation=<rot 0 -86.4832 -85.8875>
  }]
 dyn_rock_54 [{
   shapeType=Capsule
   length=1.5168
   radius=0.5319
   shapeName=dyn_rock_54_capsule
   center=<1.5082 -0.2612 -0.0272>
   rotation=<rot 0 91.2189 -89.1247>
  }]
 dyn_rock_55 [{
   shapeType=Capsule
   length=2.2279
   radius=0.8652
   shapeName=dyn_rock_55_capsule
   center=<2.2096 -0.1183 0.0762>
   rotation=<rot 0 91.7527 -89.0237>
  }]
 dyn_rock_56 [{
   shapeType=Capsule
   length=3.9635
   radius=1.5214
   shapeName=dyn_rock_56_capsule
   center=<4.0056 -0.3905 -0.0406>
   rotation=<rot 0 -83.4107 -85.2581>
  }]
 dyn_rock_57 [{
   shapeType=Capsule
   length=2.7143
   radius=1.0065
   shapeName=dyn_rock_57_capsule
   center=<2.734 -0.2265 -0.058>
   rotation=<rot 0 -83.3982 -86.0735>
  }]
 dyn_rock_58 [{
   shapeType=Capsule
   length=3.2932
   radius=1.1879
   shapeName=dyn_rock_58_capsule
   center=<3.0338 0.2205 -0.2664>
   rotation=<rot 0 87.8748 -83.4739>
  }]
 dyn_rock_59 [{
   shapeType=Capsule
   length=1.0446
   radius=0.5129
   shapeName=dyn_rock_59_capsule
   center=<1.0287 -0.3534 -0.0671>
   rotation=<rot 0 87.1659 -82.1231>
  }]
 dyn_rock_60 [{
   shapeType=Capsule
   length=5.3137
   radius=1.6386
   shapeName=dyn_rock_60_capsule
   center=<5.3123 0.538 -0.9688>
   rotation=<rot 0 92.7927 -86.4098>
  }]
 dyn_rocksBase_01 [{
   shapeType=Capsule
   length=1.4567
   radius=0.9625
   shapeName=dyn_rocksBase_01_capsule
   center=<3.2917 5.796 2.6203>
   rotation=<rot 0 149.9237 -80.5759>
  }]
 dyn_rocksBase_02 [{shapeType=Capsule, length=0.505, radius=0.505, shapeName=dyn_rocksBase_02_capsule, rotation=<rot 0 -90 0>}]
 dyn_rocksBase_03 [{shapeType=Capsule, length=0.505, radius=0.505, shapeName=dyn_rocksBase_03_capsule, rotation=<rot 0 -90 0>}]
 dyn_rocksBase_04 [{shapeType=Capsule, length=0.505, radius=0.505, shapeName=dyn_rocksBase_04_capsule, rotation=<rot 0 -90 0>}]
}
constraints [
 {bone1Name=dyn_rock_03, bone2Name=dyn_rocksBase_01, swing1LimitDegrees=10, swing2LimitDegrees=10, twistLimitDegrees=2}
 {bone1Name=dyn_rock_07, bone2Name=dyn_rocksBase_01, swing1LimitDegrees=10, swing2LimitDegrees=10, twistLimitDegrees=2}
 {bone1Name=dyn_rock_06, bone2Name=dyn_rocksBase_01, swing1LimitDegrees=10, swing2LimitDegrees=10, twistLimitDegrees=2}
 {bone1Name=dyn_rock_01, bone2Name=dyn_rocksBase_01, swing1LimitDegrees=10, swing2LimitDegrees=10, twistLimitDegrees=2}
 {bone1Name=dyn_rock_08, bone2Name=dyn_rocksBase_01, swing1LimitDegrees=10, swing2LimitDegrees=10, twistLimitDegrees=2}
 {bone1Name=dyn_rock_02, bone2Name=dyn_rocksBase_01, swing1LimitDegrees=10, swing2LimitDegrees=10, twistLimitDegrees=2}
 {bone1Name=dyn_rock_17, bone2Name=dyn_rocksBase_01, swing1LimitDegrees=10, swing2LimitDegrees=10, twistLimitDegrees=2}
 {bone1Name=dyn_rock_59, bone2Name=dyn_rocksBase_01, swing1LimitDegrees=10, swing2LimitDegrees=10, twistLimitDegrees=2}
 {bone1Name=dyn_rock_10, bone2Name=dyn_rocksBase_01, swing1LimitDegrees=10, swing2LimitDegrees=10, twistLimitDegrees=2}
 {bone1Name=dyn_rock_11, bone2Name=dyn_rocksBase_01, swing1LimitDegrees=10, swing2LimitDegrees=10, twistLimitDegrees=2}
 {bone1Name=dyn_rock_14, bone2Name=dyn_rocksBase_01, swing1LimitDegrees=10, swing2LimitDegrees=10, twistLimitDegrees=2}
 {bone1Name=dyn_rock_05, bone2Name=dyn_rocksBase_01, swing1LimitDegrees=10, swing2LimitDegrees=10, twistLimitDegrees=2}
 {bone1Name=dyn_rock_37, bone2Name=dyn_rocksBase_01, swing1LimitDegrees=10, swing2LimitDegrees=10, twistLimitDegrees=2}
 {bone1Name=dyn_rock_43, bone2Name=dyn_rocksBase_01, swing1LimitDegrees=10, swing2LimitDegrees=10, twistLimitDegrees=2}
 {bone1Name=dyn_rock_38, bone2Name=dyn_rocksBase_01, swing1LimitDegrees=10, swing2LimitDegrees=10, twistLimitDegrees=2}
 {bone1Name=dyn_rock_42, bone2Name=dyn_rocksBase_01, swing1LimitDegrees=10, swing2LimitDegrees=10, twistLimitDegrees=2}
 {bone1Name=dyn_rock_39, bone2Name=dyn_rocksBase_01, swing1LimitDegrees=10, swing2LimitDegrees=10, twistLimitDegrees=2}
 {bone1Name=dyn_rock_40, bone2Name=dyn_rocksBase_01, swing1LimitDegrees=10, swing2LimitDegrees=10, twistLimitDegrees=2}
 {bone1Name=dyn_rock_44, bone2Name=dyn_rocksBase_01, swing1LimitDegrees=10, swing2LimitDegrees=10, twistLimitDegrees=2}
 {bone1Name=dyn_rock_41, bone2Name=dyn_rocksBase_01, swing1LimitDegrees=10, swing2LimitDegrees=10, twistLimitDegrees=2}
 {bone1Name=dyn_rock_12, bone2Name=dyn_rocksBase_01, swing1LimitDegrees=10, swing2LimitDegrees=10, twistLimitDegrees=2}
 {bone1Name=dyn_rock_13, bone2Name=dyn_rocksBase_01, swing1LimitDegrees=10, swing2LimitDegrees=10, twistLimitDegrees=2}
 {bone1Name=dyn_rock_15, bone2Name=dyn_rocksBase_01, swing1LimitDegrees=10, swing2LimitDegrees=10, twistLimitDegrees=2}
 {bone1Name=dyn_rock_16, bone2Name=dyn_rocksBase_01, swing1LimitDegrees=10, swing2LimitDegrees=10, twistLimitDegrees=2}
 {bone1Name=dyn_rock_09, bone2Name=dyn_rocksBase_01, swing1LimitDegrees=10, swing2LimitDegrees=10, twistLimitDegrees=2}
 {bone1Name=dyn_rock_04, bone2Name=dyn_rocksBase_01, swing1LimitDegrees=10, swing2LimitDegrees=10, twistLimitDegrees=2}
 {bone1Name=dyn_rock_18, bone2Name=dyn_rocksBase_01, swing1LimitDegrees=10, swing2LimitDegrees=10, twistLimitDegrees=2}
 {bone1Name=dyn_rock_34, bone2Name=dyn_rocksBase_01, swing1LimitDegrees=10, swing2LimitDegrees=10, twistLimitDegrees=2}
 {bone1Name=dyn_rock_19, bone2Name=dyn_rocksBase_01, swing1LimitDegrees=10, swing2LimitDegrees=10, twistLimitDegrees=2}
 {bone1Name=dyn_rock_22, bone2Name=dyn_rocksBase_01, swing1LimitDegrees=10, swing2LimitDegrees=10, twistLimitDegrees=2}
 {bone1Name=dyn_rock_27, bone2Name=dyn_rocksBase_01, swing1LimitDegrees=10, swing2LimitDegrees=10, twistLimitDegrees=2}
 {bone1Name=dyn_rock_26, bone2Name=dyn_rocksBase_01, swing1LimitDegrees=10, swing2LimitDegrees=10, twistLimitDegrees=2}
 {bone1Name=dyn_rock_31, bone2Name=dyn_rocksBase_01, swing1LimitDegrees=10, swing2LimitDegrees=10, twistLimitDegrees=2}
 {bone1Name=dyn_rock_30, bone2Name=dyn_rocksBase_01, swing1LimitDegrees=10, swing2LimitDegrees=10, twistLimitDegrees=2}
 {bone1Name=dyn_rock_25, bone2Name=dyn_rocksBase_01, swing1LimitDegrees=10, swing2LimitDegrees=10, twistLimitDegrees=2}
 {bone1Name=dyn_rock_35, bone2Name=dyn_rocksBase_01, swing1LimitDegrees=10, swing2LimitDegrees=10, twistLimitDegrees=2}
 {bone1Name=dyn_rock_60, bone2Name=dyn_rocksBase_01, swing1LimitDegrees=10, swing2LimitDegrees=10, twistLimitDegrees=2}
 {bone1Name=dyn_rock_21, bone2Name=dyn_rocksBase_01, swing1LimitDegrees=10, swing2LimitDegrees=10, twistLimitDegrees=2}
 {bone1Name=dyn_rock_36, bone2Name=dyn_rocksBase_01, swing1LimitDegrees=10, swing2LimitDegrees=10, twistLimitDegrees=2}
 {bone1Name=dyn_rock_33, bone2Name=dyn_rocksBase_01, swing1LimitDegrees=10, swing2LimitDegrees=10, twistLimitDegrees=2}
 {bone1Name=dyn_rock_24, bone2Name=dyn_rocksBase_01, swing1LimitDegrees=10, swing2LimitDegrees=10, twistLimitDegrees=2}
 {bone1Name=dyn_rock_29, bone2Name=dyn_rocksBase_01, swing1LimitDegrees=10, swing2LimitDegrees=10, twistLimitDegrees=2}
 {bone1Name=dyn_rock_32, bone2Name=dyn_rocksBase_01, swing1LimitDegrees=10, swing2LimitDegrees=10, twistLimitDegrees=2}
 {bone1Name=dyn_rock_20, bone2Name=dyn_rocksBase_01, swing1LimitDegrees=10, swing2LimitDegrees=10, twistLimitDegrees=2}
 {bone1Name=dyn_rock_23, bone2Name=dyn_rocksBase_02, swing1LimitDegrees=10, swing2LimitDegrees=10, twistLimitDegrees=2}
 {bone1Name=dyn_rock_45, bone2Name=dyn_rocksBase_02, swing1LimitDegrees=10, swing2LimitDegrees=10, twistLimitDegrees=2}
 {bone1Name=dyn_rock_47, bone2Name=dyn_rocksBase_03, swing1LimitDegrees=10, swing2LimitDegrees=10, twistLimitDegrees=2}
 {bone1Name=dyn_rock_46, bone2Name=dyn_rocksBase_03, swing1LimitDegrees=10, swing2LimitDegrees=10, twistLimitDegrees=2}
 {bone1Name=dyn_rock_53, bone2Name=dyn_rocksBase_03, swing1LimitDegrees=10, swing2LimitDegrees=10, twistLimitDegrees=2}
 {bone1Name=dyn_rock_55, bone2Name=dyn_rocksBase_03, swing1LimitDegrees=10, swing2LimitDegrees=10, twistLimitDegrees=2}
 {bone1Name=dyn_rock_58, bone2Name=dyn_rocksBase_03, swing1LimitDegrees=10, swing2LimitDegrees=10, twistLimitDegrees=2}
 {bone1Name=dyn_rock_56, bone2Name=dyn_rocksBase_03, swing1LimitDegrees=10, swing2LimitDegrees=10, twistLimitDegrees=2}
 {bone1Name=dyn_rock_57, bone2Name=dyn_rocksBase_03, swing1LimitDegrees=10, swing2LimitDegrees=10, twistLimitDegrees=2}
 {bone1Name=dyn_rock_51, bone2Name=dyn_rocksBase_03, swing1LimitDegrees=10, swing2LimitDegrees=10, twistLimitDegrees=2}
 {bone1Name=dyn_rock_50, bone2Name=dyn_rocksBase_03, swing1LimitDegrees=10, swing2LimitDegrees=10, twistLimitDegrees=2}
 {bone1Name=dyn_rock_49, bone2Name=dyn_rocksBase_03, swing1LimitDegrees=10, swing2LimitDegrees=10, twistLimitDegrees=2}
 {bone1Name=dyn_rock_48, bone2Name=dyn_rocksBase_03, swing1LimitDegrees=10, swing2LimitDegrees=10, twistLimitDegrees=2}
 {bone1Name=dyn_rock_52, bone2Name=dyn_rocksBase_03, swing1LimitDegrees=10, swing2LimitDegrees=10, twistLimitDegrees=2}
 {bone1Name=dyn_rock_54, bone2Name=dyn_rocksBase_04, swing1LimitDegrees=10, swing2LimitDegrees=10, twistLimitDegrees=2}
 {bone1Name=dyn_rock_28, bone2Name=dyn_rocksBase_04, swing1LimitDegrees=10, swing2LimitDegrees=10, twistLimitDegrees=2}
]
