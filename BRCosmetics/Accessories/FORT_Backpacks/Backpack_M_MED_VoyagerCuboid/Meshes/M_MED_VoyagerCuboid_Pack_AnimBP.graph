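# Animation Blueprint for the VoyagerCuboid backpack cosmetic, posing its rock and cube bones.
# The cached input pose runs through a RigidBody simulation, switched off in the front end via FE Alpha, then a long chain of additive bone-space Modify Bone nodes.
# Each update, the EventGraph computes phase-offset sine translations, time-driven spins and an ambient drift vector from real time; their bindings to bones are not visible.
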AnimBlueprint /BRCosmetics/Accessories/FORT_Backpacks/Backpack_M_MED_VoyagerCuboid/Meshes/M_MED_VoyagerCuboid_Pack_AnimBP
target_skeleton=&/M_MED_VoyagerCuboid_Pack_Skeleton
graphs [
 AnimGraph {nodes [
   n1 = RigidBody(ComponentPose=Local To Component(LocalPose=Use cached pose 'Input_pose_cache'()), Alpha=Get FE Alpha() -> (FE Alpha:double)) {
    bOverrideWorldGravity=true
    overridePhysicsAsset=&/M_MED_VoyagerCuboid_Pack_Physics
    simSpaceSettings {
     dampingAlpha=5
     maxAngularAcceleration=40
     maxAngularVelocity=8
     maxLinearAcceleration=4000
     maxLinearVelocity=800
     velocityScaleZ=0.1
     worldAlpha=0.25
     externalLinearDragV=<0.3 3>
    }
   }
   n2 = "Transform (Modify) Bone - Bone: dyn_rock_03"(ComponentPose=@n1, Alpha=0.1, Translation="0,0,0", Rotation="0,0,0", Scale="1,1,1") {
    rotationMode=BMM_Additive
    scaleSpace=BCS_BoneSpace
    translationMode=BMM_Additive
    translationSpace=BCS_BoneSpace
    alphaBoolBlend {blendOption=Linear}
    boneToModify {boneName=dyn_rock_03}
   }
   n3 = "Transform (Modify) Bone - Bone: dyn_rock_07"!(ComponentPose=@n2, Alpha=1, Translation="1,0,0") -> (Pose) {
    rotationMode=BMM_Additive
    scaleSpace=BCS_BoneSpace
    translationMode=BMM_Additive
    translationSpace=BCS_BoneSpace
    alphaBoolBlend {blendOption=Linear}
    boneToModify {boneName=dyn_rock_07}
    translation {x=1}
   }
   n4 = "Transform (Modify) Bone - Bone: dyn_rock_06"!(ComponentPose=@n3, Alpha=1, Translation="1,0,0") -> (Pose) {
    rotationMode=BMM_Additive
    scaleSpace=BCS_BoneSpace
    translationMode=BMM_Additive
    translationSpace=BCS_BoneSpace
    alphaBoolBlend {blendOption=Linear}
    boneToModify {boneName=dyn_rock_06}
    translation {x=1}
   }
   n5 = "Transform (Modify) Bone - Bone: dyn_rock_01"!(ComponentPose=@n4, Alpha=1, Translation="1,0,0") -> (Pose) {
    rotationMode=BMM_Additive
    scaleSpace=BCS_BoneSpace
    translationMode=BMM_Additive
    translationSpace=BCS_BoneSpace
    alphaBoolBlend {blendOption=Linear}
    boneToModify {boneName=dyn_rock_01}
    translation {x=1}
   }
   n6 = "Transform (Modify) Bone - Bone: dyn_rock_08"!(ComponentPose=@n5, Alpha=1, Translation="1,0,0") -> (Pose) {
    rotationMode=BMM_Additive
    scaleSpace=BCS_BoneSpace
    translationMode=BMM_Additive
    translationSpace=BCS_BoneSpace
    alphaBoolBlend {blendOption=Linear}
    boneToModify {boneName=dyn_rock_08}
    translation {x=1}
   }
   n7 = "Transform (Modify) Bone - Bone: dyn_rock_02"!(ComponentPose=@n6, Alpha=1, Translation="1,0,0") -> (Pose) {
    rotationMode=BMM_Additive
    scaleSpace=BCS_BoneSpace
    translationMode=BMM_Additive
    translationSpace=BCS_BoneSpace
    alphaBoolBlend {blendOption=Linear}
    boneToModify {boneName=dyn_rock_02}
    translation {x=1}
   }
   n8 = "Transform (Modify) Bone - Bone: dyn_rock_17"!(ComponentPose=@n7, Alpha=1, Translation="2,0,0") -> (Pose) {
    rotationMode=BMM_Additive
    scaleSpace=BCS_BoneSpace
    translationMode=BMM_Additive
    translationSpace=BCS_BoneSpace
    alphaBoolBlend {blendOption=Linear}
    boneToModify {boneName=dyn_rock_17}
    translation {x=2}
   }
   n9 = "Transform (Modify) Bone - Bone: dyn_rock_59"(ComponentPose=@n8, Alpha=1, Translation="0,0,0", Rotation="0,0,0", Scale="1,1,1") {
    rotationMode=BMM_Additive
    scaleSpace=BCS_BoneSpace
    translationMode=BMM_Additive
    translationSpace=BCS_BoneSpace
    alphaBoolBlend {blendOption=Linear}
    boneToModify {boneName=dyn_rock_59}
   }
   n10 = "Transform (Modify) Bone - Bone: dyn_rock_10"(ComponentPose=@n9, Alpha=1, Translation="0,0,0", Rotation="0,0,0", Scale="1,1,1") {
    rotationMode=BMM_Additive
    scaleSpace=BCS_BoneSpace
    translationMode=BMM_Additive
    translationSpace=BCS_BoneSpace
    alphaBoolBlend {blendOption=Linear}
    boneToModify {boneName=dyn_rock_10}
   }
   n11 = "Transform (Modify) Bone - Bone: dyn_rock_11"(ComponentPose=@n10, Alpha=1, Translation="0,0,0", Rotation="0,0,0", Scale="1,1,1") {
    rotationMode=BMM_Additive
    scaleSpace=BCS_BoneSpace
    translationMode=BMM_Additive
    translationSpace=BCS_BoneSpace
    alphaBoolBlend {blendOption=Linear}
    boneToModify {boneName=dyn_rock_11}
   }
   n12 = "Transform (Modify) Bone - Bone: dyn_rock_14"(ComponentPose=@n11, Alpha=1, Translation="0,0,0", Rotation="0,0,0", Scale="1,1,1") {
    rotationMode=BMM_Additive
    scaleSpace=BCS_BoneSpace
    translationMode=BMM_Additive
    translationSpace=BCS_BoneSpace
    alphaBoolBlend {blendOption=Linear}
    boneToModify {boneName=dyn_rock_14}
   }
   n13 = "Transform (Modify) Bone - Bone: dyn_rock_05"(ComponentPose=@n12, Alpha=1, Translation="0,0,0", Rotation="0,0,0", Scale="1,1,1") {
    rotationMode=BMM_Additive
    scaleSpace=BCS_BoneSpace
    translationMode=BMM_Additive
    translationSpace=BCS_BoneSpace
    alphaBoolBlend {blendOption=Linear}
    boneToModify {boneName=dyn_rock_05}
   }
   n14 = "Transform (Modify) Bone - Bone: dyn_rock_37"(ComponentPose=@n13, Alpha=1, Translation="0,0,0", Rotation="0,0,0", Scale="1,1,1") {
    rotationMode=BMM_Additive
    scaleSpace=BCS_BoneSpace
    translationMode=BMM_Additive
    translationSpace=BCS_BoneSpace
    alphaBoolBlend {blendOption=Linear}
    boneToModify {boneName=dyn_rock_37}
   }
   n15 = "Transform (Modify) Bone - Bone: dyn_rock_43"(ComponentPose=@n14, Alpha=1, Translation="0,0,0", Rotation="0,0,0", Scale="1,1,1") {
    rotationMode=BMM_Additive
    scaleSpace=BCS_BoneSpace
    translationMode=BMM_Additive
    translationSpace=BCS_BoneSpace
    alphaBoolBlend {blendOption=Linear}
    boneToModify {boneName=dyn_rock_43}
   }
   n16 = "Transform (Modify) Bone - Bone: dyn_rock_38"(ComponentPose=@n15, Alpha=1, Translation="0,0,0", Rotation="0,0,0", Scale="1,1,1") {
    rotationMode=BMM_Additive
    scaleSpace=BCS_BoneSpace
    translationMode=BMM_Additive
    translationSpace=BCS_BoneSpace
    alphaBoolBlend {blendOption=Linear}
    boneToModify {boneName=dyn_rock_38}
   }
   n17 = "Transform (Modify) Bone - Bone: dyn_rock_42"(ComponentPose=@n16, Alpha=1, Translation="0,0,0", Rotation="0,0,0", Scale="1,1,1") {
    rotationMode=BMM_Additive
    scaleSpace=BCS_BoneSpace
    translationMode=BMM_Additive
    translationSpace=BCS_BoneSpace
    alphaBoolBlend {blendOption=Linear}
    boneToModify {boneName=dyn_rock_42}
   }
   n18 = "Transform (Modify) Bone - Bone: dyn_rock_39"(ComponentPose=@n17, Alpha=1, Translation="0,0,0", Rotation="0,0,0", Scale="1,1,1") {
    rotationMode=BMM_Additive
    scaleSpace=BCS_BoneSpace
    translationMode=BMM_Additive
    translationSpace=BCS_BoneSpace
    alphaBoolBlend {blendOption=Linear}
    boneToModify {boneName=dyn_rock_39}
   }
   n19 = "Transform (Modify) Bone - Bone: dyn_rock_40"(ComponentPose=@n18, Alpha=1, Translation="0,0,0", Rotation="0,0,0", Scale="1,1,1") {
    rotationMode=BMM_Additive
    scaleSpace=BCS_BoneSpace
    translationMode=BMM_Additive
    translationSpace=BCS_BoneSpace
    alphaBoolBlend {blendOption=Linear}
    boneToModify {boneName=dyn_rock_40}
   }
   n20 = "Transform (Modify) Bone - Bone: dyn_rock_44"(ComponentPose=@n19, Alpha=1, Translation="0,0,0", Rotation="0,0,0", Scale="1,1,1") {
    rotationMode=BMM_Additive
    scaleSpace=BCS_BoneSpace
    translationMode=BMM_Additive
    translationSpace=BCS_BoneSpace
    alphaBoolBlend {blendOption=Linear}
    boneToModify {boneName=dyn_rock_44}
   }
   n21 = "Transform (Modify) Bone - Bone: dyn_rock_41"(ComponentPose=@n20, Alpha=1, Translation="0,0,0", Rotation="0,0,0", Scale="1,1,1") {
    rotationMode=BMM_Additive
    scaleSpace=BCS_BoneSpace
    translationMode=BMM_Additive
    translationSpace=BCS_BoneSpace
    alphaBoolBlend {blendOption=Linear}
    boneToModify {boneName=dyn_rock_41}
   }
   n22 = "Transform (Modify) Bone - Bone: dyn_rock_12"(ComponentPose=@n21, Alpha=1, Translation="0,0,0", Rotation="0,0,0", Scale="1,1,1") {
    rotationMode=BMM_Additive
    scaleSpace=BCS_BoneSpace
    translationMode=BMM_Additive
    translationSpace=BCS_BoneSpace
    alphaBoolBlend {blendOption=Linear}
    boneToModify {boneName=dyn_rock_12}
   }
   n23 = "Transform (Modify) Bone - Bone: dyn_rock_13"(ComponentPose=@n22, Alpha=1, Translation="0,0,0", Rotation="0,0,0", Scale="1,1,1") {
    rotationMode=BMM_Additive
    scaleSpace=BCS_BoneSpace
    translationMode=BMM_Additive
    translationSpace=BCS_BoneSpace
    alphaBoolBlend {blendOption=Linear}
    boneToModify {boneName=dyn_rock_13}
   }
   n24 = "Transform (Modify) Bone - Bone: dyn_rock_15"(ComponentPose=@n23, Alpha=1, Translation="0,0,0", Rotation="0,0,0", Scale="1,1,1") {
    rotationMode=BMM_Additive
    scaleSpace=BCS_BoneSpace
    translationMode=BMM_Additive
    translationSpace=BCS_BoneSpace
    alphaBoolBlend {blendOption=Linear}
    boneToModify {boneName=dyn_rock_15}
   }
   n25 = "Transform (Modify) Bone - Bone: dyn_rock_16"(ComponentPose=@n24, Alpha=1, Translation="0,0,0", Rotation="0,0,0", Scale="1,1,1") {
    rotationMode=BMM_Additive
    scaleSpace=BCS_BoneSpace
    translationMode=BMM_Additive
    translationSpace=BCS_BoneSpace
    alphaBoolBlend {blendOption=Linear}
    boneToModify {boneName=dyn_rock_16}
   }
   n26 = "Transform (Modify) Bone - Bone: dyn_rock_09"(ComponentPose=@n25, Alpha=1, Translation="0,0,0", Rotation="0,0,0", Scale="1,1,1") {
    rotationMode=BMM_Additive
    scaleSpace=BCS_BoneSpace
    translationMode=BMM_Additive
    translationSpace=BCS_BoneSpace
    alphaBoolBlend {blendOption=Linear}
    boneToModify {boneName=dyn_rock_09}
   }
   n27 = "Transform (Modify) Bone - Bone: dyn_rock_04"(ComponentPose=@n26, Alpha=1, Translation="0,0,0", Rotation="0,0,0", Scale="1,1,1") {
    rotationMode=BMM_Additive
    scaleSpace=BCS_BoneSpace
    translationMode=BMM_Additive
    translationSpace=BCS_BoneSpace
    alphaBoolBlend {blendOption=Linear}
    boneToModify {boneName=dyn_rock_04}
   }
   n28 = "Transform (Modify) Bone - Bone: dyn_rock_18"(ComponentPose=@n27, Alpha=1, Translation="0,0,0", Rotation="0,0,0", Scale="1,1,1") {
    rotationMode=BMM_Additive
    scaleSpace=BCS_BoneSpace
    translationMode=BMM_Additive
    translationSpace=BCS_BoneSpace
    alphaBoolBlend {blendOption=Linear}
    boneToModify {boneName=dyn_rock_18}
   }
   n29 = "Transform (Modify) Bone - Bone: dyn_rock_34"(ComponentPose=@n28, Alpha=1, Translation="0,0,0", Rotation="0,0,0", Scale="1,1,1") {
    rotationMode=BMM_Additive
    scaleSpace=BCS_BoneSpace
    translationMode=BMM_Additive
    translationSpace=BCS_BoneSpace
    alphaBoolBlend {blendOption=Linear}
    boneToModify {boneName=dyn_rock_34}
   }
   n30 = "Transform (Modify) Bone - Bone: dyn_rock_19"(ComponentPose=@n29, Alpha=1, Translation="0,0,0", Rotation="0,0,0", Scale="1,1,1") {
    rotationMode=BMM_Additive
    scaleSpace=BCS_BoneSpace
    translationMode=BMM_Additive
    translationSpace=BCS_BoneSpace
    alphaBoolBlend {blendOption=Linear}
    boneToModify {boneName=dyn_rock_19}
   }
   n31 = "Transform (Modify) Bone - Bone: dyn_rock_22"(ComponentPose=@n30, Alpha=1, Translation="0,0,0", Rotation="0,0,0", Scale="1,1,1") {
    rotationMode=BMM_Additive
    scaleSpace=BCS_BoneSpace
    translationMode=BMM_Additive
    translationSpace=BCS_BoneSpace
    alphaBoolBlend {blendOption=Linear}
    boneToModify {boneName=dyn_rock_22}
   }
   n32 = "Transform (Modify) Bone - Bone: dyn_rock_27"(ComponentPose=@n31, Alpha=1, Translation="0,0,0", Rotation="0,0,0", Scale="1,1,1") {
    rotationMode=BMM_Additive
    scaleSpace=BCS_BoneSpace
    translationMode=BMM_Additive
    translationSpace=BCS_BoneSpace
    alphaBoolBlend {blendOption=Linear}
    boneToModify {boneName=dyn_rock_27}
   }
   n33 = "Transform (Modify) Bone - Bone: dyn_rock_26"(ComponentPose=@n32, Alpha=1, Translation="0,0,0", Rotation="0,0,0", Scale="1,1,1") {
    rotationMode=BMM_Additive
    scaleSpace=BCS_BoneSpace
    translationMode=BMM_Additive
    translationSpace=BCS_BoneSpace
    alphaBoolBlend {blendOption=Linear}
    boneToModify {boneName=dyn_rock_26}
   }
   n34 = "Transform (Modify) Bone - Bone: dyn_rock_31"(ComponentPose=@n33, Alpha=1, Translation="0,0,0", Rotation="0,0,0", Scale="1,1,1") {
    rotationMode=BMM_Additive
    scaleSpace=BCS_BoneSpace
    translationMode=BMM_Additive
    translationSpace=BCS_BoneSpace
    alphaBoolBlend {blendOption=Linear}
    boneToModify {boneName=dyn_rock_31}
   }
   n35 = "Transform (Modify) Bone - Bone: dyn_rock_30"(ComponentPose=@n34, Alpha=1, Translation="0,0,0", Rotation="0,0,0", Scale="1,1,1") {
    rotationMode=BMM_Additive
    scaleSpace=BCS_BoneSpace
    translationMode=BMM_Additive
    translationSpace=BCS_BoneSpace
    alphaBoolBlend {blendOption=Linear}
    boneToModify {boneName=dyn_rock_30}
   }
   n36 = "Transform (Modify) Bone - Bone: dyn_rock_25"(ComponentPose=@n35, Alpha=1, Translation="0,0,0", Rotation="0,0,0", Scale="1,1,1") {
    rotationMode=BMM_Additive
    scaleSpace=BCS_BoneSpace
    translationMode=BMM_Additive
    translationSpace=BCS_BoneSpace
    alphaBoolBlend {blendOption=Linear}
    boneToModify {boneName=dyn_rock_25}
   }
   n37 = "Transform (Modify) Bone - Bone: dyn_rock_35"(ComponentPose=@n36, Alpha=1, Translation="0,0,0", Rotation="0,0,0", Scale="1,1,1") {
    rotationMode=BMM_Additive
    scaleSpace=BCS_BoneSpace
    translationMode=BMM_Additive
    translationSpace=BCS_BoneSpace
    alphaBoolBlend {blendOption=Linear}
    boneToModify {boneName=dyn_rock_35}
   }
   n38 = "Transform (Modify) Bone - Bone: dyn_rock_60"(ComponentPose=@n37, Alpha=1, Translation="0,0,0", Rotation="0,0,0", Scale="1,1,1") {
    rotationMode=BMM_Additive
    scaleSpace=BCS_BoneSpace
    translationMode=BMM_Additive
    translationSpace=BCS_BoneSpace
    alphaBoolBlend {blendOption=Linear}
    boneToModify {boneName=dyn_rock_60}
   }
   n39 = "Transform (Modify) Bone - Bone: dyn_rock_21"(ComponentPose=@n38, Alpha=1, Translation="0,0,0", Rotation="0,0,0", Scale="1,1,1") {
    rotationMode=BMM_Additive
    scaleSpace=BCS_BoneSpace
    translationMode=BMM_Additive
    translationSpace=BCS_BoneSpace
    alphaBoolBlend {blendOption=Linear}
    boneToModify {boneName=dyn_rock_21}
   }
   n40 = "Transform (Modify) Bone - Bone: dyn_rock_36"(ComponentPose=@n39, Alpha=1, Translation="0,0,0", Rotation="0,0,0", Scale="1,1,1") {
    rotationMode=BMM_Additive
    scaleSpace=BCS_BoneSpace
    translationMode=BMM_Additive
    translationSpace=BCS_BoneSpace
    alphaBoolBlend {blendOption=Linear}
    boneToModify {boneName=dyn_rock_36}
   }
   n41 = "Transform (Modify) Bone - Bone: dyn_rock_33"(ComponentPose=@n40, Alpha=1, Translation="0,0,0", Rotation="0,0,0", Scale="1,1,1") {
    rotationMode=BMM_Additive
    scaleSpace=BCS_BoneSpace
    translationMode=BMM_Additive
    translationSpace=BCS_BoneSpace
    alphaBoolBlend {blendOption=Linear}
    boneToModify {boneName=dyn_rock_33}
   }
   n42 = "Transform (Modify) Bone - Bone: dyn_rock_24"(ComponentPose=@n41, Alpha=1, Translation="0,0,0", Rotation="0,0,0", Scale="1,1,1") {
    rotationMode=BMM_Additive
    scaleSpace=BCS_BoneSpace
    translationMode=BMM_Additive
    translationSpace=BCS_BoneSpace
    alphaBoolBlend {blendOption=Linear}
    boneToModify {boneName=dyn_rock_24}
   }
   n43 = "Transform (Modify) Bone - Bone: dyn_rock_29"(ComponentPose=@n42, Alpha=1, Translation="0,0,0", Rotation="0,0,0", Scale="1,1,1") {
    rotationMode=BMM_Additive
    scaleSpace=BCS_BoneSpace
    translationMode=BMM_Additive
    translationSpace=BCS_BoneSpace
    alphaBoolBlend {blendOption=Linear}
    boneToModify {boneName=dyn_rock_29}
   }
   n44 = "Transform (Modify) Bone - Bone: dyn_rock_32"(ComponentPose=@n43, Alpha=1, Translation="0,0,0", Rotation="0,0,0", Scale="1,1,1") {
    rotationMode=BMM_Additive
    scaleSpace=BCS_BoneSpace
    translationMode=BMM_Additive
    translationSpace=BCS_BoneSpace
    alphaBoolBlend {blendOption=Linear}
    boneToModify {boneName=dyn_rock_32}
   }
   n45 = "Transform (Modify) Bone - Bone: dyn_rock_20"(ComponentPose=@n44, Alpha=1, Translation="0,0,0", Rotation="0,0,0", Scale="1,1,1") {
    rotationMode=BMM_Additive
    scaleSpace=BCS_BoneSpace
    translationMode=BMM_Additive
    translationSpace=BCS_BoneSpace
    alphaBoolBlend {blendOption=Linear}
    boneToModify {boneName=dyn_rock_20}
   }
   n46 = "Transform (Modify) Bone - Bone: dyn_rock_47"!(ComponentPose=@n45, Alpha=1, Translation="1,0,0") -> (Pose) {
    rotationMode=BMM_Additive
    scaleSpace=BCS_BoneSpace
    translationMode=BMM_Additive
    translationSpace=BCS_BoneSpace
    alphaBoolBlend {blendOption=Linear}
    boneToModify {boneName=dyn_rock_47}
    translation {x=1}
   }
   n47 = "Transform (Modify) Bone - Bone: dyn_rock_46"!(ComponentPose=@n46, Alpha=1, Translation="1,0,0") -> (Pose) {
    rotationMode=BMM_Additive
    scaleSpace=BCS_BoneSpace
    translationMode=BMM_Additive
    translationSpace=BCS_BoneSpace
    alphaBoolBlend {blendOption=Linear}
    boneToModify {boneName=dyn_rock_46}
    translation {x=1}
   }
   n48 = "Transform (Modify) Bone - Bone: dyn_rock_53"!(ComponentPose=@n47, Alpha=1, Translation="1,0,0") -> (Pose) {
    rotationMode=BMM_Additive
    scaleSpace=BCS_BoneSpace
    translationMode=BMM_Additive
    translationSpace=BCS_BoneSpace
    alphaBoolBlend {blendOption=Linear}
    boneToModify {boneName=dyn_rock_53}
    translation {x=1}
   }
   n49 = "Transform (Modify) Bone - Bone: dyn_rock_55"!(ComponentPose=@n48, Alpha=1, Translation="1,0,0") -> (Pose) {
    rotationMode=BMM_Additive
    scaleSpace=BCS_BoneSpace
    translationMode=BMM_Additive
    translationSpace=BCS_BoneSpace
    alphaBoolBlend {blendOption=Linear}
    boneToModify {boneName=dyn_rock_55}
    translation {x=1}
   }
   n50 = "Transform (Modify) Bone - Bone: dyn_rock_58"!(ComponentPose=@n49, Alpha=1, Translation="1,0,0") -> (Pose) {
    rotationMode=BMM_Additive
    scaleSpace=BCS_BoneSpace
    translationMode=BMM_Additive
    translationSpace=BCS_BoneSpace
    alphaBoolBlend {blendOption=Linear}
    boneToModify {boneName=dyn_rock_58}
    translation {x=1}
   }
   n51 = "Transform (Modify) Bone - Bone: dyn_rock_56"!(ComponentPose=@n50, Alpha=1, Translation="2,0,0") -> (Pose) {
    rotationMode=BMM_Additive
    scaleSpace=BCS_BoneSpace
    translationMode=BMM_Additive
    translationSpace=BCS_BoneSpace
    alphaBoolBlend {blendOption=Linear}
    boneToModify {boneName=dyn_rock_56}
    translation {x=2}
   }
   n52 = "Transform (Modify) Bone - Bone: dyn_rock_57"(ComponentPose=@n51, Alpha=1, Translation="0,0,0", Rotation="0,0,0", Scale="1,1,1") {
    rotationMode=BMM_Additive
    scaleSpace=BCS_BoneSpace
    translationMode=BMM_Additive
    translationSpace=BCS_BoneSpace
    alphaBoolBlend {blendOption=Linear}
    boneToModify {boneName=dyn_rock_57}
   }
   n53 = "Transform (Modify) Bone - Bone: dyn_rock_51"(ComponentPose=@n52, Alpha=1, Translation="0,0,0", Rotation="0,0,0", Scale="1,1,1") {
    rotationMode=BMM_Additive
    scaleSpace=BCS_BoneSpace
    translationMode=BMM_Additive
    translationSpace=BCS_BoneSpace
    alphaBoolBlend {blendOption=Linear}
    boneToModify {boneName=dyn_rock_51}
   }
   n54 = "Transform (Modify) Bone - Bone: dyn_rock_50"(ComponentPose=@n53, Alpha=1, Translation="0,0,0", Rotation="0,0,0", Scale="1,1,1") {
    rotationMode=BMM_Additive
    scaleSpace=BCS_BoneSpace
    translationMode=BMM_Additive
    translationSpace=BCS_BoneSpace
    alphaBoolBlend {blendOption=Linear}
    boneToModify {boneName=dyn_rock_50}
   }
   n55 = "Transform (Modify) Bone - Bone: dyn_rock_49"(ComponentPose=@n54, Alpha=1, Translation="0,0,0", Rotation="0,0,0", Scale="1,1,1") {
    rotationMode=BMM_Additive
    scaleSpace=BCS_BoneSpace
    translationMode=BMM_Additive
    translationSpace=BCS_BoneSpace
    alphaBoolBlend {blendOption=Linear}
    boneToModify {boneName=dyn_rock_49}
   }
   n56 = "Transform (Modify) Bone - Bone: dyn_rock_48"(ComponentPose=@n55, Alpha=1, Translation="0,0,0", Rotation="0,0,0", Scale="1,1,1") {
    rotationMode=BMM_Additive
    scaleSpace=BCS_BoneSpace
    translationMode=BMM_Additive
    translationSpace=BCS_BoneSpace
    alphaBoolBlend {blendOption=Linear}
    boneToModify {boneName=dyn_rock_48}
   }
   n57 = "Transform (Modify) Bone - Bone: dyn_rock_52"(ComponentPose=@n56, Alpha=1, Translation="0,0,0", Rotation="0,0,0", Scale="1,1,1") {
    rotationMode=BMM_Additive
    scaleSpace=BCS_BoneSpace
    translationMode=BMM_Additive
    translationSpace=BCS_BoneSpace
    alphaBoolBlend {blendOption=Linear}
    boneToModify {boneName=dyn_rock_52}
   }
   n58 = "Transform (Modify) Bone - Bone: dyn_rock_23"(ComponentPose=@n57, Alpha=1, Translation="0,0,0", Rotation="0,0,0", Scale="1,1,1") {
    rotationMode=BMM_Additive
    scaleSpace=BCS_BoneSpace
    translationMode=BMM_Additive
    translationSpace=BCS_BoneSpace
    alphaBoolBlend {blendOption=Linear}
    boneToModify {boneName=dyn_rock_23}
   }
   n59 = "Transform (Modify) Bone - Bone: dyn_rock_45"(ComponentPose=@n58, Alpha=1, Translation="0,0,0", Rotation="0,0,0", Scale="1,1,1") {
    rotationMode=BMM_Additive
    scaleSpace=BCS_BoneSpace
    translationMode=BMM_Additive
    translationSpace=BCS_BoneSpace
    alphaBoolBlend {blendOption=Linear}
    boneToModify {boneName=dyn_rock_45}
   }
   n60 = "Transform (Modify) Bone - Bone: dyn_rock_54"(ComponentPose=@n59, Alpha=1, Translation="0,0,0", Rotation="0,0,0", Scale="1,1,1") {
    rotationMode=BMM_Additive
    scaleSpace=BCS_BoneSpace
    translationMode=BMM_Additive
    translationSpace=BCS_BoneSpace
    alphaBoolBlend {blendOption=Linear}
    boneToModify {boneName=dyn_rock_54}
   }
   n61 = "Transform (Modify) Bone - Bone: dyn_rock_28"(ComponentPose=@n60, Alpha=1, Translation="0,0,0", Rotation="0,0,0", Scale="1,1,1") {
    rotationMode=BMM_Additive
    scaleSpace=BCS_BoneSpace
    translationMode=BMM_Additive
    translationSpace=BCS_BoneSpace
    alphaBoolBlend {blendOption=Linear}
    boneToModify {boneName=dyn_rock_28}
   }
   n62 = "Transform (Modify) Bone - Bone: dyn_rocksBase_01"(ComponentPose=@n61, Alpha=1, Translation="0,0,0", Rotation="0,0,0", Scale="1,1,1") {
    rotationMode=BMM_Additive
    scaleSpace=BCS_BoneSpace
    translationMode=BMM_Additive
    translationSpace=BCS_BoneSpace
    alphaBoolBlend {blendOption=Linear}
    boneToModify {boneName=dyn_rocksBase_01}
   }
   n63 = "Transform (Modify) Bone - Bone: dyn_rocksBase_02"(ComponentPose=@n62, Alpha=1, Translation="0,0,0", Rotation="0,0,0", Scale="1,1,1") {
    rotationMode=BMM_Additive
    scaleSpace=BCS_BoneSpace
    translationMode=BMM_Additive
    translationSpace=BCS_BoneSpace
    alphaBoolBlend {blendOption=Linear}
    boneToModify {boneName=dyn_rocksBase_02}
   }
   n64 = "Transform (Modify) Bone - Bone: dyn_rocksBase_03"(ComponentPose=@n63, Alpha=1, Translation="0,0,0", Rotation="0,0,0", Scale="1,1,1") {
    rotationMode=BMM_Additive
    scaleSpace=BCS_BoneSpace
    translationMode=BMM_Additive
    translationSpace=BCS_BoneSpace
    alphaBoolBlend {blendOption=Linear}
    boneToModify {boneName=dyn_rocksBase_03}
   }
   n65 = "Transform (Modify) Bone - Bone: dyn_cube_01"(ComponentPose=@n64, Alpha=1, Translation="0,0,0", Rotation="0,0,0", Scale="1,1,1") {
    rotationMode=BMM_Additive
    scaleSpace=BCS_BoneSpace
    translationMode=BMM_Additive
    translationSpace=BCS_BoneSpace
    alphaBoolBlend {blendOption=Linear}
    boneToModify {boneName=dyn_cube_01}
   }
   n66 = "Transform (Modify) Bone - Bone: dyn_cube_02"(ComponentPose=@n65, Alpha=1, Translation="0,0,0", Rotation="0,0,0", Scale="1,1,1") {
    rotationMode=BMM_Additive
    scaleSpace=BCS_BoneSpace
    translationMode=BMM_Additive
    translationSpace=BCS_BoneSpace
    alphaBoolBlend {blendOption=Linear}
    boneToModify {boneName=dyn_cube_02}
   }
   n67 = "Transform (Modify) Bone - Bone: dyn_cube_03"(ComponentPose=@n66, Alpha=1, Translation="0,0,0", Rotation="0,0,0", Scale="1,1,1") {
    rotationMode=BMM_Additive
    scaleSpace=BCS_BoneSpace
    translationMode=BMM_Additive
    translationSpace=BCS_BoneSpace
    alphaBoolBlend {blendOption=Linear}
    boneToModify {boneName=dyn_cube_03}
   }
   n68 = "Transform (Modify) Bone - Bone: dyn_cube_04"(ComponentPose=@n67, Alpha=1, Translation="0,0,0", Rotation="0,0,0", Scale="1,1,1") {
    rotationMode=BMM_Additive
    scaleSpace=BCS_BoneSpace
    translationMode=BMM_Additive
    translationSpace=BCS_BoneSpace
    alphaBoolBlend {blendOption=Linear}
    boneToModify {boneName=dyn_cube_04}
   }
   n69 = "Transform (Modify) Bone - Bone: dyn_cube_05"(ComponentPose=@n68, Alpha=1, Translation="0,0,0", Rotation="0,0,0", Scale="1,1,1") {
    rotationMode=BMM_Additive
    scaleSpace=BCS_BoneSpace
    translationMode=BMM_Additive
    translationSpace=BCS_BoneSpace
    alphaBoolBlend {blendOption=Linear}
    boneToModify {boneName=dyn_cube_05}
   }
   n70 = "Transform (Modify) Bone - Bone: dyn_cube_06"(ComponentPose=@n69, Alpha=1, Translation="0,0,0", Rotation="0,0,0", Scale="1,1,1") {
    rotationMode=BMM_Additive
    scaleSpace=BCS_BoneSpace
    translationMode=BMM_Additive
    translationSpace=BCS_BoneSpace
    alphaBoolBlend {blendOption=Linear}
    boneToModify {boneName=dyn_cube_06}
   }
   Output Pose(Result=Component To Local(ComponentPose=@n70)) {layerGroup=None}
   Save cached pose 'Input_pose_cache'(Pose=Input Pose()) {cacheName=Input_pose_cache}
  ]}
 CalculateVibrations {nodes [
   n1 = Inputs!() -> (Frequency:double=~, Amplitude:double=~, Noise:double=~)
   n2 = RandomFloatInRange(Min=float * float(A=@n1.Amplitude, B=-1) {is=title K2Node_PromotableOperator}, Max=@n1.Noise)
   n3 = GetGameTimeInSeconds()
   n4 = float * float(A="Sin (Radians)"(A=float * float(A=@n1, B=@n3)), B=@n1.Amplitude)
   Outputs!(Result:double="Clamp (Float)"(Value=float + float(A=@n4, B=@n2), Min=-1, Max=10))
  ]}
 EventGraph {nodes [
   Event BlueprintUpdateAnimation()
   > n1 = Set Range(Range:double=2) -> (Output_Get:double)
   > Set alpha(alpha:double=0.75) -> (Output_Get:double)
   n2 = GetRealTimeSeconds()
   n3 = "Sin (Radians)"(A=float * float(A=float - float(A=@n2, B=0.6), B=2.5) {is=title K2Node_CommutativeAssociativeBinaryOperator})
   > "Set Translate Shard 1 "("Translate Shard 1 ":Vector=MakeVector(X=float / float(A=@n3, B=@n1) {is=title K2Node_CallFunction}, Y=0)) -> (Output_Get:Vector)
   n4 = GetRealTimeSeconds()
   n5 = "Sin (Radians)"(A=float * float(A=float - float(A=@n4, B=0.2), B=2.5))
   > Set Translate Shard 2(Translate Shard 2:Vector=MakeVector(X=float / float(A=@n5, B=@n1), Y=0)) -> (Output_Get:Vector)
   n6 = GetRealTimeSeconds()
   n7 = "Sin (Radians)"(A=float * float(A=float - float(A=@n6, B=0.4), B=2.5))
   > Set Translate Shard 3(Translate Shard 3:Vector=MakeVector(X=float / float(A=@n7, B=@n1), Y=0)) -> (Output_Get:Vector)
   n8 = GetRealTimeSeconds()
   n9 = "Sin (Radians)"(A=float * float(A=float - float(A=@n8, B=0.6), B=2.5))
   > Set Translate Shard 4(Translate Shard 4:Vector=MakeVector(X=float / float(A=@n9, B=@n1), Y=0)) -> (Output_Get:Vector)
   n10 = GetRealTimeSeconds()
   n11 = "Sin (Radians)"(A=float * float(A=float - float(A=@n10, B=0.8), B=2.5))
   > Set Translate Shard 5(Translate Shard 5:Vector=MakeVector(X=float / float(A=@n11, B=@n1), Y=0)) -> (Output_Get:Vector)
   n12 = GetRealTimeSeconds()
   n13 = "Sin (Radians)"(A=float * float(A=float - float(A=@n12, B=1), B=2.5))
   > Set Translate Shard 6(Translate Shard 6:Vector=MakeVector(X=float / float(A=@n13, B=@n1), Y=0)) -> (Output_Get:Vector)
   n14 = GetRealTimeSeconds()
   n15 = "Sin (Radians)"(A=float * float(A=float - float(A=@n14, B=1), B=3.5))
   > Set Translate Shard Large 1(Translate Shard Large 1:Vector=MakeVector(X=float / float(A=@n15, B=@n1), Y=0)) -> (Output_Get:Vector)
   n16 = GetRealTimeSeconds()
   n17 = "Sin (Radians)"(A=float * float(A=float - float(A=@n16, B=0.75), B=3.5))
   > Set Translate Shard Large 2(Translate Shard Large 2:Vector=MakeVector(X=float / float(A=@n17, B=@n1), Y=0)) -> (Output_Get:Vector)
   n18 = GetRealTimeSeconds()
   n19 = "Sin (Radians)"(A=float * float(A=float - float(A=@n18, B=0.5), B=3.5))
   > Set Translate Shard Large 3(Translate Shard Large 3:Vector=MakeVector(X=float / float(A=@n19, B=@n1), Y=0)) -> (Output_Get:Vector)
   n20 = float - float(A=GetRealTimeSeconds(), B=1)
   > Set Rotate Shard Positive(Rotate Shard Positive:Rotator=MakeRotator(Pitch=float * float(A=@n20, B=2.5))) -> (Output_Get:Rotator)
   n21 = float - float(A=GetRealTimeSeconds(), B=1)
   > Set Rotate Shard Negative(Rotate Shard Negative:Rotator=MakeRotator(Pitch=float * float(A=@n21, B=-2.5))) -> (Output_Get:Rotator)
   n22 = float - float(A=GetRealTimeSeconds(), B=1)
   > Set Rotate Shard Fast(Rotate Shard Fast:Rotator=MakeRotator(Pitch=float * float(A=@n22, B=-5))) -> (Output_Get:Rotator)
   n23 = float - float(A=GetRealTimeSeconds(), B=1)
   > Set Rotate Shard Cube(Rotate Shard Cube:Rotator=MakeRotator(Pitch=float * float(A=@n23, B=-20))) -> (Output_Get:Rotator)
   n24 = float * float(A=float / float(A=GetRealTimeSeconds(), B=-720), B=93750)
   n25 = float * float(A=@n24, B=0.8)
   > Set Rotate Shard Cube_01(Rotate Shard Cube_01:Rotator=MakeRotator(Roll=@n25, Pitch=@n25, Yaw=@n25)) -> (Output_Get:Rotator)
   n26 = float * float(A=@n24, B=0.9)
   > Set Rotate Shard Cube_02(Rotate Shard Cube_02:Rotator=MakeRotator(Roll=@n26, Pitch=@n26, Yaw=@n26)) -> (Output_Get:Rotator)
   n27 = float * float(A=@n24, B=1)
   > Set Rotate Shard Cube_03(Rotate Shard Cube_03:Rotator=MakeRotator(Roll=@n27, Pitch=@n27, Yaw=@n27)) -> (Output_Get:Rotator)
   n28 = float * float(A=@n24, B=1.1)
   > Set Rotate Shard Cube_04(Rotate Shard Cube_04:Rotator=MakeRotator(Roll=@n28, Pitch=@n28, Yaw=@n28)) -> (Output_Get:Rotator)
   n29 = float * float(A=@n24, B=1.2)
   > Set Rotate Shard Cube_05(Rotate Shard Cube_05:Rotator=MakeRotator(Roll=@n29, Pitch=@n29, Yaw=@n29)) -> (Output_Get:Rotator)
   n30 = float * float(A=@n24, B=1.3)
   > Set Rotate Shard Cube_06(Rotate Shard Cube_06:Rotator=MakeRotator(Roll=@n30, Pitch=@n30, Yaw=@n30)) -> (Output_Get:Rotator)
   n31 = GetRealTimeSeconds()
   n32 = float * float(A="Sin (Radians)"(A=float * float(A=1.2, B=@n31)), B=0.2)
   n33 = float * float(A="Sin (Radians)"(A=float * float(A=1, B=@n31)), B=0.3)
   n34 = Remap Float(Input="Sin (Radians)"(A=float * float(A=2, B=@n31)), OldRange=(X=-1,Y=1), NewRange=(X=-1,Y=2)) -> (Output=~)
   n35 = float * float(A="Clamp (Float)"(Value=@n34, Min=-1, Max=0), B=0.5)
   > Set Ambient Translation(Ambient Translation:Vector=vector + vector(A=MakeVector(X=@n32, Y=@n33, Z=@n34), B=MakeVector(Y=@n35))) -> (Output_Get:Vector)
   > n36 = Branch(Condition=Get bIsInFrontEnd())
   > Set FE Alpha(FE Alpha:double=0) -> (Output_Get:double)
   Set FE Alpha(execute=@n36.else, FE Alpha=1)
   TryGetPawnOwner()
   TryGetPawnOwner()
  ]}
]
variables [
 alpha {type=number}
 Ambient Translation {type=Vector}
 FE Alpha {type=number}
 Range {type=number}
 Rotate Shard Cube {type=Rotator}
 Rotate Shard Cube_01 {type=Rotator}
 Rotate Shard Cube_02 {type=Rotator}
 Rotate Shard Cube_03 {type=Rotator}
 Rotate Shard Cube_04 {type=Rotator}
 Rotate Shard Cube_05 {type=Rotator}
 Rotate Shard Cube_06 {type=Rotator}
 Rotate Shard Fast {type=Rotator}
 Rotate Shard Negative {type=Rotator}
 Rotate Shard Positive {type=Rotator}
 "Translate Shard 1 " {type=Vector}
 Translate Shard 2 {type=Vector}
 Translate Shard 3 {type=Vector}
 Translate Shard 4 {type=Vector}
 Translate Shard 5 {type=Vector}
 Translate Shard 6 {type=Vector}
 Translate Shard Large 1 {type=Vector}
 Translate Shard Large 2 {type=Vector}
 Translate Shard Large 3 {type=Vector}
]
